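; TODM_Disabled is a time-of-day manager Blueprint variant that, by its name, stands in for a disabled day/night setup.
; Its construction script hides the DayNightCycle directional light component without propagating to children.
; On BeginPlay it destroys the editor post-process material component, a hack that works around code injecting the outline shader at runtime.
; The Tick event is empty.
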(event EventBeginPlay
  ;[1 1 1] destroy hacky post process at run time
  (Components|DestroyComponent (Variables|Default|GetEditorPostProcessMaterial(HackToWorkAroundTheOutlineShaderBeingInjectedByCodeAtRuntime))))

(event EventTick
  (params (DeltaSeconds :type Float%20%28single-precision%29)))

(fn ConstructionScript
  (Rendering|SetVisibility (Variables|DayNightCycle|GetDirectionalLightComponent) false false))
